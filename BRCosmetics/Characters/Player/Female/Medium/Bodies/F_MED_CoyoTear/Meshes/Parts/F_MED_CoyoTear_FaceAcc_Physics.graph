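# Physics asset for the CoyoTear female character's face accessory mesh, defining a ponytail chain and the body shapes around it.
# Four capsule bodies form the ponytail, linked in sequence to an AIM body on the head with limited swing and tight twist.
# Capsules on the head, spine, upper arms and a pack collision bone provide surrounding volumes, joined by constraints with free swing and twist.
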PhysicsAsset /BRCosmetics/Characters/Player/Female/Medium/Bodies/F_MED_CoyoTear/Meshes/Parts/F_MED_CoyoTear_FaceAcc_Physics
bodies {
 dyn_pack_collision [{shapeType=Capsule, length=9.005, radius=15.505, shapeName=dyn_ponytail_MSTR_capsule, center=<0 15 -2>, rotation=<rot 90 0 0>}]
 dyn_ponytail_01 [{shapeType=Capsule, length=7.4111, radius=5.1808, shapeName=dyn_ponytail_01_capsule, center=<5 0 2>, rotation=<rot 86.4878 0 0>}]
 dyn_ponytail_02 [{shapeType=Capsule, length=7.4359, radius=3.6923, shapeName=dyn_ponytail_02_capsule, center=<7 -0.5 0>, rotation=<rot 90 0 0>}]
 dyn_ponytail_03 [{shapeType=Capsule, length=7.2937, radius=2.5227, shapeName=dyn_ponytail_03_capsule, center=<5 -0.5 0>, rotation=<rot 90 0 0>}]
 dyn_ponytail_04 [{shapeType=Capsule, length=7.3831, radius=3.5214, shapeName=dyn_ponytail_04_capsule, center=<6.9351 0 0>, rotation=<rot 90 0 0>}]
 dyn_ponytail_AIM [{shapeType=Capsule, length=0.505, radius=0.505, shapeName=dyn_ponytail_MSTR_capsule, rotation=<rot 0 -90 0>}]
 head [{
   shapeType=Capsule
   length=6.005
   radius=10.505
   shapeName=dyn_ponytail_MSTR_capsule
   center=<3.4327 1.1868 0>
   rotation=<rot 90 1.4321 11.2042>
  }]
 spine_02 [{shapeType=Capsule, length=4.005, radius=12.505, shapeName=dyn_ponytail_MSTR_capsule, center=<4.2808 5.553 0>, rotation=<rot 90 90 81.496>}]
 spine_05 [{shapeType=Capsule, length=3.505, radius=17.005, shapeName=dyn_ponytail_MSTR_capsule, center=<4.1123 8.5122 0>, rotation=<rot 0 -90 0>}]
 upperarm_l [{
   shapeType=Capsule
   length=19.505
   radius=7.005
   shapeName=dyn_ponytail_MSTR_capsule
   center=<10.2556 0.5437 4.2759>
   rotation=<rot 88.7976 -7.0298 179.9943>
  }]
 upperarm_r [{shapeType=Capsule, length=24.505, radius=6.005, shapeName=dyn_ponytail_MSTR_capsule, center=<-12.1012 0.4691 0>, rotation=<rot 90 0 0>}]
}
constraints [
 {bone1Name=dyn_ponytail_01, bone2Name=dyn_ponytail_AIM, swing1LimitDegrees=45, swing2LimitDegrees=30, twistLimitDegrees=5}
 {bone1Name=dyn_ponytail_02, bone2Name=dyn_ponytail_01, swing1LimitDegrees=45, swing2LimitDegrees=45, twistLimitDegrees=5}
 {bone1Name=dyn_ponytail_03, bone2Name=dyn_ponytail_02, swing1LimitDegrees=45, swing2LimitDegrees=45, twistLimitDegrees=5}
 {bone1Name=dyn_ponytail_04, bone2Name=dyn_ponytail_03, swing1LimitDegrees=45, swing2LimitDegrees=45, twistLimitDegrees=5}
 {
  bone1Name=upperarm_r
  bone2Name=spine_05
  swing1LimitDegrees=45
  swing1Motion=Free
  swing2LimitDegrees=45
  swing2Motion=Free
  twistLimitDegrees=45
  twistMotion=Free
 }
 {
  bone1Name=upperarm_l
  bone2Name=spine_05
  swing1LimitDegrees=45
  swing1Motion=Free
  swing2LimitDegrees=45
  swing2Motion=Free
  twistLimitDegrees=45
  twistMotion=Free
 }
 {
  bone1Name=spine_05
  bone2Name=spine_02
  swing1LimitDegrees=45
  swing1Motion=Free
  swing2LimitDegrees=45
  swing2Motion=Free
  twistLimitDegrees=45
  twistMotion=Free
 }
 {
  bone1Name=dyn_ponytail_AIM
  bone2Name=head
  swing1LimitDegrees=45
  swing1Motion=Free
  swing2LimitDegrees=45
  swing2Motion=Free
  twistLimitDegrees=45
  twistMotion=Free
 }
 {
  bone1Name=dyn_pack_collision
  bone2Name=spine_05
  swing1LimitDegrees=45
  swing1Motion=Free
  swing2LimitDegrees=45
  swing2Motion=Free
  twistLimitDegrees=45
  twistMotion=Free
 }
 {
  bone1Name=head
  bone2Name=spine_05
  swing1LimitDegrees=45
  swing1Motion=Free
  swing2LimitDegrees=45
  swing2Motion=Free
  twistLimitDegrees=45
  twistMotion=Free
 }
]
