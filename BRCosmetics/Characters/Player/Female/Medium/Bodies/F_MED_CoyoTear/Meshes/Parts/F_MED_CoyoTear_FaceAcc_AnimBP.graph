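# Post-process style AnimBP for CoyoTear's face accessory mesh, adding secondary motion to hair, ponytail and earrings.
# Copies the attached parent's pose through a hair Control Rig, then AnimDynamics, bending right-side hair while targeting, into cached basePose.
# Without a backpack, basePose also gets ponytail LookAt and neck-relative RigidBody simulation, with collision offset for capes; backpacks bypass this.
# Customization load sets backpack and cape flags from backbling.
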AnimBlueprint /BRCosmetics/Characters/Player/Female/Medium/Bodies/F_MED_CoyoTear/Meshes/Parts/F_MED_CoyoTear_FaceAcc_AnimBP
target_skeleton=&/F_MED_CoyoTear_FaceAcc_Skeleton
graphs [
 AnimGraph {nodes [
   n1 = "Transform (Modify) Bone - Bone: dyn_pack_collision"!(
    ComponentPose=Local To Component(LocalPose=Use cached pose 'basePose'())
    bAlphaBoolEnabled:bool=Get isWearingCape() -> (isWearingCape:bool)
    Translation="0,-23,-3"
    Rotation="0,0,0"
    Scale="1,1,1"
   ) -> (Pose) {
    alphaInputType=Bool
    rotationMode=BMM_Additive
    rotationSpace=BCS_BoneSpace
    translationMode=BMM_Additive
    translationSpace=BCS_BoneSpace
    alphaBoolBlend {blendOption=Linear}
    boneToModify {boneName=dyn_pack_collision}
    translation {y=-23, z=-3}
   }
   n2 = Look At - Bone: dyn_ponytail_AIM(ComponentPose=@n1, Alpha=1) {
    alphaBoolBlend {blendOption=Linear}
    boneToModify {boneName=dyn_ponytail_AIM}
    lookAtLocation=<-20 -66 0>
    lookAtTarget {boneReference {boneName=spine_05}}
    lookAt_Axis {axis=<1 0>}
   }
   n3 = RigidBody!(ComponentPose=@n2, Alpha, OverrideWorldGravity:Vector="0,0,-980", ExternalForce, ComponentLinearAccScale:Vector="0.35,0.35,0.35") -> (Pose) {
    bClampLinearTranslationLimitToRefPose=true
    bOverrideWorldGravity=true
    overridePhysicsAsset=&/F_MED_CoyoTear_FaceAcc_Physics
    simulationSpace=BaseBoneSpace
    alphaBoolBlend {blendOption=Linear}
    alphaScaleBiasClamp {interpSpeedDecreasing=4, interpSpeedIncreasing=6}
    baseBoneRef {boneName=neck_01}
    componentAppliedLinearAccClamp=<800 800 800>
    componentLinearAccScale=<0.6 0.6 0.2>
    componentLinearVelScale=<0.5 0.5 0.5>
    overrideWorldGravity {z=-875}
    simSpaceSettings {maxAngularAcceleration=200, maxAngularVelocity=5, maxLinearAcceleration=500, maxLinearVelocity=200}
   }
   Output Pose(Result=Blend Poses by bool(
     BlendPose_0=Use cached pose 'basePose'()
     BlendPose_1=Component To Local(ComponentPose=@n3)
     bActiveValue=Get isWearingBackpack() -> (isWearingBackpack:bool)
    ) {blendType=Linear}) {layerGroup=None}
   n4 = Control Rig!(Source=Copy Pose From Mesh() {bUseAttachedParent=true}, Alpha, isWearingBackpack:bool=Get isWearingBackpack()) -> (Pose) {
    alphaBoolBlend {blendOption=Linear}
    controlRigAssetReference {blueprintRigClass=&/F_MED_CoyoTear_Hair_CtrlRig.F_MED_CoyoTear_Hair_CtrlRig_C}
   }
   n5 = Anim Dynamics - Bone: dyn_earring_l(ComponentPose=Local To Component(LocalPose=@n4), Alpha=1, ExternalForce="0,0,0") {
    angularDampingOverride=0.999
    angularSpringConstant=500
    bAngularSpring=true
    bOverrideAngularDamping=true
    bUseGravityOverride=true
    simulationSpace=RootRelative
    alphaBoolBlend {blendOption=Linear}
    boundBone {boneName=dyn_earring_l}
    componentAppliedLinearAccClamp=<5000 5000 5000>
    componentLinearAccScale=<0.3 0.3 0.3>
    componentLinearVelScale=<0.3 0.3 0.3>
    gravityOverride {z=-980}
    physicsBodyDefinitions [{
      boundBone {boneName=dyn_earring_l}
      boxExtents=<20 15 15>
      constraintSetup {angularLimitsMax=<0 5 30>, angularLimitsMin=<0 -5 -30>, angularTarget=<0 0 0>}
      localJointOffset=<4 0 0>
     }]
   }
   n6 = Anim Dynamics - Bone: dyn_earring_r(ComponentPose=@n5, Alpha=1, ExternalForce="0,0,0") {
    angularDampingOverride=0.999
    angularSpringConstant=500
    bAngularSpring=true
    bOverrideAngularDamping=true
    bUseGravityOverride=true
    simulationSpace=RootRelative
    alphaBoolBlend {blendOption=Linear}
    boundBone {boneName=dyn_earring_r}
    componentAppliedLinearAccClamp=<5000 5000 5000>
    componentLinearAccScale=<0.3 0.3 0.3>
    componentLinearVelScale=<0.3 0.3 0.3>
    gravityOverride {z=-980}
    physicsBodyDefinitions [{
      boundBone {boneName=dyn_earring_r}
      boxExtents=<20 15 15>
      constraintSetup {angularLimitsMax=<0 5 30>, angularLimitsMin=<0 -5 -30>, angularTarget=<0 0 0>}
      localJointOffset=<4 0 0>
     }]
   }
   n7 = "Anim Dynamics - Chain: dyn_hair_l_01 -> dyn_hair_l_02"!(ComponentPose=@n6, bAlphaBoolEnabled:bool=True, ExternalForce="0,0,0") -> (Pose) {
    alphaInputType=Bool
    angularDampingOverride=0.999
    angularSpringConstant=500
    bChain=true
    bOverrideAngularDamping=true
    bUseGravityOverride=true
    simulationSpace=RootRelative
    alphaBoolBlend {blendOption=Linear}
    boundBone {boneName=dyn_hair_l_01}
    chainEnd {boneName=dyn_hair_l_02}
    componentAppliedLinearAccClamp=<5000 5000 5000>
    componentLinearAccScale=<0.3 0.3 0.3>
    componentLinearVelScale=<0.3 0.3 0.3>
    gravityOverride {z=-980}
    physicsBodyDefinitions [
     {
      boundBone {boneName=dyn_hair_l_01}
      boxExtents=<15 20 20>
      constraintSetup {angularLimitsMax=<0 5 5>, angularLimitsMin=<0 0 -5>, angularTarget=<0 0 0>}
      localJointOffset=<2 0 1>
     }
     {
      boundBone {boneName=dyn_hair_l_02}
      boxExtents=<10 15 15>
      constraintSetup {angularLimitsMax=<0 10 5>, angularLimitsMin=<0 -2 -5>, angularTarget=<0 0 0>}
      localJointOffset=<1 0 1.5>
     }
    ]
   }
   n8 = "Anim Dynamics - Chain: dyn_hair_r_01 -> dyn_hair_r_02"!(ComponentPose=@n7, bAlphaBoolEnabled:bool=True, ExternalForce="0,0,0") -> (Pose) {
    alphaInputType=Bool
    angularDampingOverride=0.999
    angularSpringConstant=500
    bChain=true
    bOverrideAngularDamping=true
    bUseGravityOverride=true
    simulationSpace=RootRelative
    alphaBoolBlend {blendOption=Linear, blendOutTime=0.3}
    boundBone {boneName=dyn_hair_r_01}
    chainEnd {boneName=dyn_hair_r_02}
    componentAppliedLinearAccClamp=<5000 5000 5000>
    componentLinearAccScale=<0.3 0.3 0.3>
    componentLinearVelScale=<0.3 0.3 0.3>
    gravityOverride {z=-980}
    physicsBodyDefinitions [
     {
      boundBone {boneName=dyn_hair_r_01}
      boxExtents=<15 20 20>
      constraintSetup {angularLimitsMax=<0 5 5>, angularLimitsMin=<0 0 -5>, angularTarget=<0 0 0>}
      localJointOffset=<2 0 1>
     }
     {
      boundBone {boneName=dyn_hair_r_02}
      boxExtents=<10 15 15>
      constraintSetup {angularLimitsMax=<0 10 5>, angularLimitsMin=<0 -2 -5>, angularTarget=<0 0 0>}
      localJointOffset=<1 0 1.5>
     }
    ]
   }
   n9 = Anim Dynamics - Bone: dyn_hair_bk_01(ComponentPose=@n8, Alpha=1, ExternalForce="0,0,0") {
    angularDampingOverride=0.999
    angularSpringConstant=500
    bAngularSpring=true
    bGravityOverrideInSimSpace=true
    bOverrideAngularDamping=true
    bUseGravityOverride=true
    simulationSpace=BoneRelative
    alphaBoolBlend {blendOption=Linear}
    boundBone {boneName=dyn_hair_bk_01}
    componentAppliedLinearAccClamp=<5000 5000 5000>
    componentLinearAccScale=<0.3 0.3 0.3>
    componentLinearVelScale=<0.1 0.1 0.1>
    gravityOverride {x=500, z=-180}
    physicsBodyDefinitions [{
      boundBone {boneName=dyn_hair_bk_01}
      boxExtents=<15 20 20>
      constraintSetup {angularLimitsMax=<0 0 10>, angularLimitsMin=<0 -30 -5>, angularTarget=<0 0 0>}
      localJointOffset=<2 0 0>
     }]
    relativeSpaceBone {boneName=dyn_ponytail_MSTR}
   }
   n10 = Anim Dynamics - Bone: dyn_hair_bk_02(ComponentPose=@n9, Alpha=1, ExternalForce="0,0,0") {
    angularDampingOverride=0.99
    angularSpringConstant=800
    bAngularSpring=true
    bGravityOverrideInSimSpace=true
    bOverrideAngularDamping=true
    bUseGravityOverride=true
    simulationSpace=BoneRelative
    alphaBoolBlend {blendOption=Linear}
    boundBone {boneName=dyn_hair_bk_02}
    componentAppliedLinearAccClamp=<5000 5000 5000>
    componentLinearAccScale=<0.1 0.1 0.1>
    componentLinearVelScale=<0.1 0.1 0.1>
    gravityOverride {x=980}
    physicsBodyDefinitions [{
      boundBone {boneName=dyn_hair_bk_02}
      boxExtents=<15 30 15>
      constraintSetup {angularLimitsMax=<0 3 15>, angularLimitsMin=<0 -20 -8>, angularTarget=<0 0 0>}
      localJointOffset=<5 0 0>
     }]
    relativeSpaceBone {boneName=dyn_ponytail_MSTR}
   }
   n11 = Get bIsTargeting()
   n12 = "Transform (Modify) Bone - Bone: dyn_hair_r_01"!(ComponentPose=@n10, bAlphaBoolEnabled:bool=@n11, Translation="0,0,-0.5", Rotation="-8,0,0", Scale="1,1,1") -> (Pose) {
    alphaInputType=Bool
    rotationMode=BMM_Additive
    rotationSpace=BCS_BoneSpace
    translationMode=BMM_Additive
    translationSpace=BCS_BoneSpace
    alphaBoolBlend {blendOption=Linear, blendOutTime=0.3}
    boneToModify {boneName=dyn_hair_r_01}
    rotation {pitch=-8}
    translation {z=-0.5}
   }
   n13 = "Transform (Modify) Bone - Bone: dyn_hair_r_02"!(ComponentPose=@n12, bAlphaBoolEnabled:bool=@n11, Translation="0,0,0", Rotation="-15,0,0", Scale="1,1,1") -> (Pose) {
    alphaInputType=Bool
    rotationMode=BMM_Additive
    rotationSpace=BCS_BoneSpace
    alphaBoolBlend {blendOption=Linear, blendOutTime=0.3}
    boneToModify {boneName=dyn_hair_r_02}
    rotation {pitch=-15}
   }
   Save cached pose 'basePose'(Pose=Component To Local(ComponentPose=@n13)) {cacheName=basePose}
  ]}
 EventGraph {nodes [
   Event OnCharacterCustomizationLoaded()
   n1 = CheckForBackbling()
   > Set isWearingBackpack(isWearingBackpack:bool=AND Boolean(A=@n1, B=NOT Boolean(A=@n1.bIsBackblingACape))) -> (Output_Get:bool)
   > Set isWearingCape(isWearingCape=@n1.bIsBackblingACape)
  ]}
]
variables [
 isWearingBackpack {type=boolean}
 isWearingCape {type=boolean}
]
